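# Sound cue for looping night ambience inside a big urban building foundation, spatialized with square map attenuation and routed to an in-building ambience sound class.
# The first quality level mixes a quieter construction-metal loop with a cave tonal loop; the second plays only the tonal loop, and the other two are empty.
# Endlessly looping 30-second envelopers slowly shape each layer, with modulators adding slight pitch and volume variation.
# A concurrency override prevents new instances once the limit is reached.
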SoundCue /FigmentEnvironment/Sounds/Fort_Ambient_Looping/Foundation_Mixes/FOUND_BuildingBig_Urban02_Night_Cue
root=QualityLevel_0
nodes [
 {
  id=Enveloper_0
  children [WavePlayer_5]
  properties {
   bLoop=true
   bLoopIndefinitely=true
   loopEnd=30
   pitchCurve {editorCurveData {keys [{interpMode=RCIM_Cubic, time=0}, <at 15.8492 0.9 RCIM_Cubic>, {interpMode=RCIM_Cubic, time=30}]}}
   volumeCurve {editorCurveData {keys [<at 0 1 RCIM_Cubic>,<at 12.9201 0.8649 RCIM_Cubic>,<at 30 1 RCIM_Cubic>]}}
  }
 }
 {
  id=Enveloper_6
  children [WavePlayer_1]
  properties {
   bLoop=true
   bLoopIndefinitely=true
   loopEnd=30
   pitchCurve {editorCurveData {keys [<at 0 1.3>]}}
   volumeCurve {externalCurve=/FigmentCore/Sounds/Trans_Enveloper_Curves/Bell_30s, editorCurveData {keys [<at 0 1>]}}
  }
 }
 {
  id=Enveloper_7
  children [WavePlayer_4]
  properties {
   bLoop=true
   bLoopIndefinitely=true
   loopEnd=30
   pitchCurve {editorCurveData {keys [{time=0}]}}
   volumeCurve {editorCurveData {keys [<at 0 1 RCIM_Cubic>,<at 15 0.5 RCIM_Cubic>,<at 30 1 RCIM_Cubic>]}}
  }
 }
 {id=Mixer_2, children [Enveloper_6,Enveloper_7], properties {inputVolume [0.5,1]}}
 {id=Modulator_0, children [Enveloper_0], properties {pitchMax=1, pitchMin=0.9, volumeMax=1.1, volumeMin=0.9}}
 {id=Modulator_12, children [Mixer_2], properties {pitchMax=1.1, pitchMin=0.9, volumeMax=1.1, volumeMin=0.9}}
 {id=QualityLevel_0, children [Modulator_12,Modulator_0,null,null]}
 {
  id=WavePlayer_1
  properties {bLooping=true, soundWaveAssetPtr=/FigmentEnvironment/Sounds/Trans_AmbientLoops/StructureMovement/CounstructionMetal027_Loop_Stereo}
 }
 {id=WavePlayer_4, properties {bLooping=true, soundWaveAssetPtr=/FigmentEnvironment/Sounds/Fort_Ambient_Looping/Cave_Mixes/Cave_Tonal11}}
 {id=WavePlayer_5, properties {bLooping=true, soundWaveAssetPtr=&Cave_Tonal11}}
]
properties {
 attenuationSettings=/Game/Sounds/Attenuation/Fort_Map_3x3_Square
 bOverrideConcurrency=true
 maxDistance=1780
 soundClassObject=/FigmentEnvironment/Sounds/FortSoundClassesAndModes/Foundations_Ambience_InBuilding_PSM
 volumeMultiplier=0.7
 concurrencyOverrides {resolutionRule=PreventNew}
}
