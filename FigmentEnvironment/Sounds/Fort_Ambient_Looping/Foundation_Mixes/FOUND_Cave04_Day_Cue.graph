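# Looping underground cave ambience cue for foundation mixes, attenuated over a map area and routed to an underground ambience sound class.
# A quality switch selects between two populated branches, leaving two slots empty.
# The first mixes pitched-down Cave_Tonal04 with quieter castle whispers, using long looping volume envelopes with offset peaks for slow crossfading movement.
# The second plays only the tonal loop with gentle pitch and volume drift; both add random modulation.
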SoundCue /FigmentEnvironment/Sounds/Fort_Ambient_Looping/Foundation_Mixes/FOUND_Cave04_Day_Cue
root=QualityLevel_0
nodes [
 {
  id=Enveloper_0
  children [WavePlayer_2]
  properties {
   bLoop=true
   bLoopIndefinitely=true
   loopEnd=30
   pitchCurve {editorCurveData {keys [
      {interpMode=RCIM_Cubic, time=0}
      {arriveTangent=0.0002, interpMode=RCIM_Cubic, leaveTangent=0.0002, tangentMode=RCTM_User, time=7.9875, value=1.0314}
      {arriveTangent=0, interpMode=RCIM_Cubic, leaveTangent=0, tangentMode=RCTM_User, time=20.0737, value=0.9355}
      {interpMode=RCIM_Cubic, time=30}
     ]}}
   volumeCurve {editorCurveData {keys [<at 0 1 RCIM_Cubic>,<at 15.6539 0.8429 RCIM_Cubic>,<at 30 1 RCIM_Cubic>]}}
  }
 }
 {
  id=Enveloper_46
  children [WavePlayer_1]
  properties {
   bLoop=true
   bLoopIndefinitely=true
   loopEnd=100
   pitchCurve {editorCurveData {keys [<at 0 0.6>]}}
   volumeCurve {editorCurveData {keys [<at 0 0.3 RCIM_Cubic>,<at 85 1 RCIM_Cubic>,<at 100 0.3 RCIM_Cubic>]}}
  }
 }
 {
  id=Enveloper_47
  children [WavePlayer_202]
  properties {
   bLoop=true
   bLoopIndefinitely=true
   loopEnd=100
   pitchCurve {editorCurveData {keys [<at 0 0.8>]}}
   volumeCurve {editorCurveData {keys [<at 0 0.3 RCIM_Cubic>,<at 10 1 RCIM_Cubic>,<at 100 0.3 RCIM_Cubic>]}}
  }
 }
 {id=Mixer_23, children [Enveloper_46,Enveloper_47], properties {inputVolume [1,0.3]}}
 {id=Modulator_0, children [Enveloper_0], properties {pitchMax=1.1, pitchMin=0.9, volumeMax=1.1, volumeMin=1}}
 {id=Modulator_60, children [Mixer_23], properties {pitchMax=1.1, pitchMin=0.9, volumeMax=1.1, volumeMin=0.9}}
 {id=QualityLevel_0, children [Modulator_60,Modulator_0,null,null]}
 {id=WavePlayer_1, properties {bLooping=true, soundWaveAssetPtr=/FigmentEnvironment/Sounds/Fort_Ambient_Looping/Cave_Mixes/Cave_Tonal04}}
 {id=WavePlayer_2, properties {bLooping=true, soundWaveAssetPtr=&Cave_Tonal04}}
 {
  id=WavePlayer_202
  properties {bLooping=true, soundWaveAssetPtr=/FigmentEnvironment/Sounds/Trans_AmbientLoops/RoomTones/Castle_WhispersD007_Stereo}
 }
]
properties {
 attenuationSettings=/Game/Sounds/Attenuation/Fort_Map_3x3_Square
 bOverrideConcurrency=true
 maxDistance=1780
 soundClassObject=/FigmentEnvironment/Sounds/FortSoundClassesAndModes/Foundations_Ambience_Underground_PSM
 volumeMultiplier=1
 concurrencyOverrides {resolutionRule=PreventNew}
}
